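# Material function for the Pickle Stomp emote prop that layers animated conveyor-belt textures and procedural ripples onto incoming material attributes.
# Time-scrolled, wrapped V coordinates build animated UVs for the belt, and masked lerps combine base color, normals, AO and specular/metallic/roughness.
# Sphere-masked sums of radial and directional sine waves are evaluated three times at offset UVs for NormalFromFunction ripple normals.
# World position offset is written on Z only.
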MaterialFunction /BRCosmetics/Animation/Props/Gadgets_PickleStomp/Materials/MF_PickleStomp_Emote
graph {nodes [
  n1 = GetMaterialAttributes(None=NamedRerouteUsage()) {attributeGetTypes [
    #1
    #2
    #3
    #4
    #5
    #6
   ]}
  n2 = NamedRerouteUsage()
  n3 = Constant()
  n4 = SetMaterialAttributes!(
   MaterialAttributes=@n1.MaterialAttributes
   BaseColor=NamedRerouteUsage()
   Normal=NamedRerouteUsage()
   AmbientOcclusion=NamedRerouteUsage()
   Specular=ComponentMask(None=@n2) {r=true}
   Metallic=ComponentMask(None=@n2) {g=true}
   Roughness=ComponentMask(None=@n2) {b=true}
   WorldPositionOffset=MaterialFunctionCall(R=@n3, G=@n3, B=NamedRerouteUsage()) {materialFunction=&AppendMany}.RGB
  ) {attributeSetTypes [
    #1
    #2
    #3
    #4
    #5
    #6
    #7
   ]}
  FunctionOutput(None=@n4)
  NamedRerouteDeclaration(None=FunctionInput() {inputName=In_MaterialAttributes, inputType=FunctionInput_MaterialAttributes}) {name=In_MaterialAttributes, nodeColor=<rgba 0 0.0625 1 1>}
  NamedRerouteDeclaration(None=FunctionInput() {inputName=In_TechArtMask, sortPriority=1}) {name=In_TechArtMask, nodeColor=<rgba 0.3828 1 0 1>}
  n5 = NamedRerouteUsage()
  n6 = Normalize(VectorInput=Add(A=NamedRerouteUsage(), B=NamedRerouteUsage()))
  n7 = NamedRerouteUsage()
  n8 = NamedRerouteUsage()
  NamedRerouteDeclaration(None=LinearInterpolate(A=LinearInterpolate(A=@n5, B=@n6, Alpha=@n7), B=NamedRerouteUsage(), Alpha=ComponentMask(None=@n8) {r=true})) {name=Out_Normals, nodeColor=<rgba 0.1484 1 0 1>}
  n9 = TextureSampleParameter2D(UVs=NamedRerouteUsage()) {
   parameterName=AnimatedM
   samplerType=SAMPLERTYPE_Masks
   texture=/BRCosmetics/Animation/Props/Gadgets_PickleStomp/Textures/T_PickleStomp_Emote_Prop_M
  }
  NamedRerouteDeclaration(None=@n9.RGB) {name=AnimatedM, nodeColor=<rgba 0 1 0.2305 1>}
  n10 = NamedRerouteUsage()
  n11 = ScalarParameter() {defaultValue=0.005, parameterName="Height Map UV Offset (S)"}
  NamedRerouteDeclaration(None=MaterialFunctionCall(R=Add(A=@n10, B=@n11), G=@n10, B=@n10, A=@n10) {materialFunction=&AppendMany}.RGB) {name=FunctionUV2, nodeColor=<rgba 1 0.5122 0.0325 1>}
  n12 = TextureSampleParameter2D(UVs=NamedRerouteUsage()) {
   parameterName=AnimatedS
   samplerType=SAMPLERTYPE_Masks
   texture=/BRCosmetics/Animation/Props/Gadgets_PickleStomp/Textures/T_PickleStomp_Emote_Prop_S
  }
  NamedRerouteDeclaration(None=@n12.RGB) {name=AnimatedS, nodeColor=<rgba 1 0 0.1433 1>}
  n13 = NamedRerouteUsage()
  n14 = ScalarParameter() {defaultValue=0.005, parameterName="Height Map UV Offset (S)"}
  NamedRerouteDeclaration(None=MaterialFunctionCall(R=@n13, G=Add(A=@n13, B=@n14), B=@n13, A=@n13) {materialFunction=&AppendMany}.RGB) {name=FunctionUV3, nodeColor=<rgba 0.0292 1 0.3327 1>}
  NamedRerouteDeclaration(None=@n1.Normal) {name=In_Normals, nodeColor=<rgba 0 1 0.25 1>}
  NamedRerouteDeclaration(None=@n1.AmbientOcclusion) {name=In_AO, nodeColor=<rgba 0.7109 0 1 1>}
  n15 = NamedRerouteUsage()
  n16 = NamedRerouteUsage()
  NamedRerouteDeclaration(None=LinearInterpolate(A=NamedRerouteUsage(), B=ComponentMask(None=@n15) {r=true}, Alpha=ComponentMask(None=@n16) {r=true})) {name=Out_AO, nodeColor=<rgba 0 1 0.3906 1>}
  n17 = NamedRerouteUsage()
  n18 = NamedRerouteUsage()
  NamedRerouteDeclaration(None=LinearInterpolate(A=NamedRerouteUsage(), B=ComponentMask(None=@n17) {r=true}, Alpha=ComponentMask(None=@n18) {r=true})) {name=Out_SMR, nodeColor=<rgba 1 0.0703 0 1>}
  NamedRerouteDeclaration(None=MaterialFunctionCall(R=@n1.Specular, G=@n1.Specular, B=@n1.Specular) {materialFunction=&AppendMany}.RGB) {name=In_SMR, nodeColor=<rgba 0 0.25 1 1>}
  NamedRerouteDeclaration(None=TextureSampleParameter2D(UVs=NamedRerouteUsage()) {parameterName=AnimatedBaseColor, texture=/BRCosmetics/Animation/Props/Gadgets_PickleStomp/Textures/T_PickleStomp_Emote_Prop_D}.RGB) {name=AnimatedBaseColor, nodeColor=<rgba 1 0 0.7031 1>}
  n19 = MaterialFunctionCall!(
   Coordinates=TextureCoordinate()
   Height Map UV Offset=ScalarParameter() {defaultValue=0.005, parameterName="Height Map UV Offset (S)"}
   Normal Map Intensity=ScalarParameter() {defaultValue=1, parameterName="Normal Map Intensity (S)"}
   "Function(UV1)"=NamedRerouteUsage()
   "Function(UV2)"=NamedRerouteUsage()
   "Function(UV3)"=NamedRerouteUsage()
  ) {materialFunction=/Engine/Functions/Engine_MaterialFunctions03/Procedurals/NormalFromFunction}
  NamedRerouteDeclaration(None=@n19.Normal) {name=ProceduralNormalsForRipples, nodeColor=<rgba 0 0.6484 1 1>}
  n20 = NamedRerouteUsage()
  NamedRerouteDeclaration(None=LinearInterpolate(A=NamedRerouteUsage(), B=NamedRerouteUsage(), Alpha=ComponentMask(None=@n20) {r=true})) {name=ConveyerBeltNormals, nodeColor=<rgba 0.9688 1 0 1>}
  NamedRerouteDeclaration(None=@n19.UV1) {name=NFF_UV1, nodeColor=<rgba 1 0 0.0234 1>}
  NamedRerouteDeclaration(None=Constant2Vector() {g=-0.6667, r=-0.73}) {name=WaveB Direction, nodeColor=<rgba 0.125 0 1 1>}
  NamedRerouteDeclaration(None=@n19.UV2) {name=NFF_UV2, nodeColor=<rgba 0 0.3437 1 1>}
  NamedRerouteDeclaration(None=@n19.UV3) {name=NFF_UV3, nodeColor=<rgba 0.6641 1 0 1>}
  n21 = SphereMask(
   A=NamedRerouteUsage()
   B=NamedRerouteUsage()
   Radius=ScalarParameter() {defaultValue=0.33, parameterName=RippleMaskRadius}
   Hardness=ScalarParameter() {defaultValue=0.064, parameterName=RippleMaskHardness}
  ) {attenuationRadius=0.5, hardnessPercent=30}
  n22 = Add(A=SmoothStep(Value=@n21), B=Multiply(A=Time(), B=ScalarParameter() {defaultValue=1, parameterName=RadialSpeed}))
  n23 = ScalarParameter() {defaultValue=1, parameterName=RadialFreq}
  n24 = Multiply(A=Sine(None=Multiply(A=@n22, B=@n23)), B=ScalarParameter() {defaultValue=0.1, parameterName=RadialAmplitude})
  n25 = NamedRerouteUsage()
  n26 = Subtract(
   A=Multiply(A=Time(), B=ScalarParameter() {defaultValue=1, parameterName=SpeedA})
   B=Multiply(A=ComponentMask(None=@n25) {r=true}, B=ScalarParameter() {defaultValue=8.2342, parameterName=FrequencyA})
  )
  n27 = NamedRerouteUsage()
  n28 = NamedRerouteUsage()
  n29 = Subtract(
   A=Multiply(A=Time(), B=ScalarParameter() {defaultValue=0.734, parameterName=SpeedB})
   B=Multiply(A=DotProduct(A=@n27, B=@n28), B=ScalarParameter() {defaultValue=4.2342, parameterName=FrequencyB})
  )
  n30 = Add(
   A=Multiply(A=Sine(None=@n26), B=ScalarParameter() {defaultValue=0.668, parameterName=Amplitude})
   B=Multiply(A=Sine(None=@n29), B=ScalarParameter() {defaultValue=0.5, parameterName=AmplitudeB})
  )
  NamedRerouteDeclaration(None=Multiply(A=NamedRerouteDeclaration(None=@n21) {name=SphereMask_1, nodeColor=<rgba 0 1 0.6953 1>}, B=Add(A=@n24, B=@n30))) {name=WPORippleA_1, nodeColor=<rgba 1 0.8438 0 1>}
  n31 = SphereMask(
   A=NamedRerouteUsage()
   B=NamedRerouteUsage()
   Radius=ScalarParameter() {defaultValue=0.33, parameterName=RippleMaskRadius}
   Hardness=ScalarParameter() {defaultValue=0.064, parameterName=RippleMaskHardness}
  ) {attenuationRadius=0.5, hardnessPercent=30}
  n32 = Add(A=SmoothStep(Value=@n31), B=Multiply(A=Time(), B=ScalarParameter() {defaultValue=1, parameterName=RadialSpeed}))
  n33 = ScalarParameter() {defaultValue=1, parameterName=RadialFreq}
  n34 = Multiply(A=Sine(None=Multiply(A=@n32, B=@n33)), B=ScalarParameter() {defaultValue=0.1, parameterName=RadialAmplitude})
  n35 = NamedRerouteUsage()
  n36 = Subtract(
   A=Multiply(A=Time(), B=ScalarParameter() {defaultValue=1, parameterName=SpeedA})
   B=Multiply(A=ComponentMask(None=@n35) {r=true}, B=ScalarParameter() {defaultValue=8.2342, parameterName=FrequencyA})
  )
  n37 = NamedRerouteUsage()
  n38 = NamedRerouteUsage()
  n39 = Subtract(
   A=Multiply(A=Time(), B=ScalarParameter() {defaultValue=0.734, parameterName=SpeedB})
   B=Multiply(A=DotProduct(A=@n37, B=@n38), B=ScalarParameter() {defaultValue=4.2342, parameterName=FrequencyB})
  )
  n40 = Add(
   A=Multiply(A=Sine(None=@n36), B=ScalarParameter() {defaultValue=0.668, parameterName=Amplitude})
   B=Multiply(A=Sine(None=@n39), B=ScalarParameter() {defaultValue=0.5, parameterName=AmplitudeB})
  )
  NamedRerouteDeclaration(None=Multiply(A=NamedRerouteDeclaration(None=@n31) {name=SphereMask_2, nodeColor=<rgba 0 1 0.6953 1>}, B=Add(A=@n34, B=@n40))) {name=WPORippleA_2, nodeColor=<rgba 1 0.8438 0 1>}
  NamedRerouteDeclaration(None=@n1.BaseColor) {name=In_BaseColor, nodeColor=<rgba 1 0.0938 0 1>}
  n41 = NamedRerouteUsage()
  NamedRerouteDeclaration(None=LinearInterpolate(A=NamedRerouteUsage(), B=NamedRerouteUsage(), Alpha=ComponentMask(None=@n41) {r=true})) {name=Out_BaseColor, nodeColor=<rgba 0 0.2266 1 1>}
  n42 = SphereMask(
   A=NamedRerouteUsage()
   B=NamedRerouteUsage()
   Radius=ScalarParameter() {defaultValue=0.33, parameterName=RippleMaskRadius}
   Hardness=ScalarParameter() {defaultValue=0.064, parameterName=RippleMaskHardness}
  ) {attenuationRadius=0.5, hardnessPercent=30}
  n43 = Add(A=SmoothStep(Value=@n42), B=Multiply(A=Time(), B=ScalarParameter() {defaultValue=1, parameterName=RadialSpeed}))
  n44 = ScalarParameter() {defaultValue=1, parameterName=RadialFreq}
  n45 = Multiply(A=Sine(None=Multiply(A=@n43, B=@n44)), B=ScalarParameter() {defaultValue=0.1, parameterName=RadialAmplitude})
  n46 = NamedRerouteUsage()
  n47 = Subtract(
   A=Multiply(A=Time(), B=ScalarParameter() {defaultValue=1, parameterName=SpeedA})
   B=Multiply(A=ComponentMask(None=@n46) {r=true}, B=ScalarParameter() {defaultValue=8.2342, parameterName=FrequencyA})
  )
  n48 = NamedRerouteUsage()
  n49 = NamedRerouteUsage()
  n50 = Subtract(
   A=Multiply(A=Time(), B=ScalarParameter() {defaultValue=0.734, parameterName=SpeedB})
   B=Multiply(A=DotProduct(A=@n48, B=@n49), B=ScalarParameter() {defaultValue=4.2342, parameterName=FrequencyB})
  )
  n51 = Add(
   A=Multiply(A=Sine(None=@n47), B=ScalarParameter() {defaultValue=0.668, parameterName=Amplitude})
   B=Multiply(A=Sine(None=@n50), B=ScalarParameter() {defaultValue=0.5, parameterName=AmplitudeB})
  )
  NamedRerouteDeclaration(None=Multiply(A=NamedRerouteDeclaration(None=@n42) {name=SphereMask, nodeColor=<rgba 0 1 0.6953 1>}, B=Add(A=@n45, B=@n51))) {name=WPORippleA, nodeColor=<rgba 1 0.8438 0 1>}
  n52 = TextureCoordinate()
  n53 = ComponentMask(None=@n52) {g=true}
  n54 = Multiply(A=Time(), B=ScalarParameter() {defaultValue=0.1, parameterName=ConveyerBeltSpeed})
  n55 = ScalarParameter() {defaultValue=0.5, parameterName=ConveyerVHigh}
  n56 = MaterialFunctionCall(
   Input=Frac(None=Add(A=@n53, B=@n54))
   Input Low=Constant()
   Input High=Constant() {r=1}
   Target Low=ScalarParameter() {parameterName=ConveyerVLow}
   Target High=@n55
  ) {materialFunction=&RemapValueRange}
  NamedRerouteDeclaration(None=AppendVector(A=ComponentMask(None=@n52) {r=true}, B=Modulo!(A=@n56.Result, B=@n55))) {name=AnimatedUVs, nodeColor=<rgba 1 0 0.8672 1>}
  NamedRerouteDeclaration(None=ComponentMask(None=VectorParameter() {parameterName=Center, defaultValue=<rgba 0.5 0.5 0 1>}.RGB) {g=true, r=true}) {name=RippleCenter, nodeColor=<rgba 1 0.1641 0 1>}
  n57 = TextureSampleParameter2D(UVs=NamedRerouteUsage()) {
   parameterName=AnimatedNormals
   samplerType=SAMPLERTYPE_Normal
   texture=/BRCosmetics/Animation/Props/Gadgets_PickleStomp/Textures/T_PickleStomp_Emote_Prop_N
  }
  NamedRerouteDeclaration(None=@n57.RGB) {name=AnimatedNormals, nodeColor=<rgba 1 0.4922 0 1>}
  n58 = NamedRerouteUsage()
  NamedRerouteDeclaration(None=MaterialFunctionCall(R=@n58, G=@n58, B=@n58, A=@n58) {materialFunction=&AppendMany}.RGB) {name=FunctionUV1, nodeColor=<rgba 0.3125 0 1 1>}
  TextureCoordinate()
  TextureCoordinate()
  TextureCoordinate()
 ]}
